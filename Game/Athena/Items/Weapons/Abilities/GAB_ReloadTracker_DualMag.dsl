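; Gameplay ability that alternates a dual-mag weapon between fast and slow reloads, based on how many times it has been reloaded.
; On activation it uses the player pawn's current weapon; when the reload count updates, it uses the reported source weapon.
; Both paths run on authority and pass the count's even/odd parity to SetReloadType on the weapon's DualMag component.
; Debug strings print the result, showing "Slow" when the returned next-reload-fast flag is true.
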(event Ability|EventActivateAbility
  ;[0 1 0] Get player pawn and weapon
  (Ability|HasAuthority
    (:True
      (bind _asfort_player_pawn (Utilities|Casting|CastToFortPlayerPawn (Ability|GetAvatarActorfromActorInfo)))
      (bind _output_get (Variables|Default|SetPlayerPawn _asfort_player_pawn))
      (bind _output_get_2 (Variables|Default|SetPlayer%20Weapon (Variables|/Script/FortniteGame.FortPawn|GetCurrentWeapon _output_get)))
      (Utilities|IsValid _output_get_2
        (:Is%20Valid
          ;[0.3747 0.9801 1] Set if reload type is fast or slow based on how many times the weapon has been reloaded
          (bind _output_get_3 (Variables|Default|SetDualMagComponent (Actor|GetComponentbyClass (Variables|Default|GetPlayer%20Weapon) "/Game/Athena/Items/Weapons/Components/FortWeaponComponent_DualMag.FortWeaponComponent_DualMag_C")))
          (Utilities|IsValid _output_get_3
            (:Is%20Valid
              (bind _bisnextreloadfast (Class|FortWeaponComponentDualMag|SetReloadType _output_get_3 (not (Math|Conversions|ToBoolean(Integer) (Math|Integer|%25%28Integer%29 (Weapon|GetScourceWeaponTimesReloaded) 2)))))
              (if _bisnextreloadfast
                (Development|PrintString "Slow" true true "(R=1.000000,G=0.467127,B=0.000000,A=1.000000)")
                (else
                  (Development|PrintString "Fast" true true "(R=0.000000,G=1.000000,B=0.098334,A=1.000000)"))))))))))

(event Weapon|EventOnSourceWeaponTimesReloadedUpdated
  (params (SourceWeapon :type Fort%20Weapon%20Object%20Reference))
  (bind _returnvalue (Weapon|GetScourceWeaponTimesReloaded))
  ;[0 1 1] Update after reload has finished
  (Ability|HasAuthority
    (:True
      (Utilities|IsValid SourceWeapon
        (:Is%20Valid
          (Variables|Default|SetPlayer%20Weapon SourceWeapon)
          (Development|PrintString (Utilities|String|ToString(Integer) _returnvalue) true true "(R=1.000000,G=0.467127,B=0.000000,A=1.000000)")
          ;[0.3747 0.9801 1] Set if reload type is fast or slow based on how many times the weapon has been reloaded
          (bind _output_get (Variables|Default|SetDualMagComponent (Actor|GetComponentbyClass (Variables|Default|GetPlayer%20Weapon) "/Game/Athena/Items/Weapons/Components/FortWeaponComponent_DualMag.FortWeaponComponent_DualMag_C")))
          (Utilities|IsValid _output_get
            (:Is%20Valid
              (bind _bisnextreloadfast (Class|FortWeaponComponentDualMag|SetReloadType _output_get (not (Math|Conversions|ToBoolean(Integer) (Math|Integer|%25%28Integer%29 _returnvalue 2)))))
              (if _bisnextreloadfast
                (Development|PrintString "Slow" true true "(R=1.000000,G=0.467127,B=0.000000,A=1.000000)")
                (else
                  (Development|PrintString "Fast" true true "(R=0.000000,G=1.000000,B=0.098334,A=1.000000)"))))))))))
